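# Conversation database for the SolarWaffle NPC in JethroLager, defining a Star Wars-flavored "Ask about..." dialogue branch under its own entry tag.
# A "How can I help?" speech offers topic choices for Attack Plans, Blue Milk, Darth Vader and The Force, plus a Back option.
# Each topic plays a spoken reply, with The Force chaining a second line, then all converge on a shared ReturnToStart node labeled BACK.
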ConversationDatabase /JethroLager/NPCs/Tandem/SolarWaffle/Jethro_Conversation_SollarWaffle_Dialogue
entry_points [{tag=Jethro.Conversation.Entry.SolarWaffle.Dialogue, destinations [#1]}]
nodes [
 ConversationEntryPointNode_0 {
  id=#1
  class=ConversationEntryPointNode
  outputs [#2]
  properties {entryTag {tagName=Jethro.Conversation.Entry.SolarWaffle.Dialogue}}
 }
 Speech_C_0 {
  id=#2
  class=Speech_C
  outputs [
   #3
   #4
   #5
   #6
   #7
  ]
  properties {nodeName="", softTaskIcon=&T_UI_NPC_question, generalConfig {defaultMessage=How can I help?}}
  sub_nodes [ConversationChoiceNode_5 {class=ConversationChoiceNode, properties {defaultChoiceDisplayText=Ask about...}}]
 }
 Speech_C_3 {
  id=#7
  class=Speech_C
  outputs [#8]
  properties {nodeName="", softTaskIcon=&T_UI_NPC_question, generalConfig {defaultMessage=Look inside yourself.}}
  sub_nodes [ConversationChoiceNode_3 {class=ConversationChoiceNode, properties {defaultChoiceDisplayText=The Force}}]
 }
 Speech_C_4 {
  id=#8
  class=Speech_C
  outputs [#9]
  properties {nodeName="", softTaskIcon=&T_UI_NPC_question, generalConfig {defaultMessage=The answers will follow.}}
 }
 ReturnToStart_C_23 {
  id=#9
  class=ReturnToStart_C
  properties {softTaskIcon=&T_UI_NPC_Back}
  sub_nodes [ConversationChoiceNode_32 {class=ConversationChoiceNode, properties {defaultChoiceDisplayText=BACK}}]
 }
 Speech_C_2 {
  id=#6
  class=Speech_C
  outputs [#9]
  properties {
   nodeName=""
   softTaskIcon=&T_UI_NPC_question
   generalConfig {defaultMessage=Some things are best left unspoken...}
  }
  sub_nodes [ConversationChoiceNode_2 {class=ConversationChoiceNode, properties {defaultChoiceDisplayText=Darth Vader}}]
 }
 Speech_C_1 {
  id=#5
  class=Speech_C
  outputs [#9]
  properties {
   nodeName=""
   softTaskIcon=&T_UI_NPC_question
   generalConfig {defaultMessage=Why do people keep asking me about this?}
  }
  sub_nodes [ConversationChoiceNode_1 {class=ConversationChoiceNode, properties {defaultChoiceDisplayText=Blue Milk}}]
 }
 Back_C_0 {
  id=#4
  class=Back_C
  sub_nodes [ConversationChoiceNode_7 {class=ConversationChoiceNode, properties {defaultChoiceDisplayText=BACK}}]
 }
 Speech_C_7 {
  id=#3
  class=Speech_C
  outputs [#9]
  properties {
   nodeName=""
   softTaskIcon=&T_UI_NPC_question
   generalConfig {defaultMessage="We'll know when it's time. Until then, we all need to train."}
  }
  sub_nodes [ConversationChoiceNode_0 {class=ConversationChoiceNode, properties {defaultChoiceDisplayText=Attack Plans}}]
 }
]
